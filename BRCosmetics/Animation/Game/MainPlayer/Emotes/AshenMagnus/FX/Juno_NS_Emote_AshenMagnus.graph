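# Niagara effect for the Juno AshenMagnus emote: an animated targeting reticle followed by timed burst and flash accents.
# DirectionalBurst layers mesh and sprite crosshair parts with curve-driven scale, spin and waveform wiggle; User.TargetB picks visibility, size and rotation variant.
# DirectionalBurst001 adds a brief mesh pulse mid-emote, then DirectionalBurst002 and Flash001 fire a shrinking sprite and orange flare.
# Front-end modifiers shrink sizes, and preview mode adjusts reticle position.
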NiagaraSystem /BRCosmetics/Animation/Game/MainPlayer/Emotes/AshenMagnus/FX/Juno_NS_Emote_AshenMagnus
dependencies {
 usedDataInterfaces [
  &NiagaraDataInterfaceCurve
  &NiagaraDataInterfaceArrayInt32
  &NiagaraDataInterfaceArrayFloat
  &NiagaraDataInterfaceColorCurve
  &NiagaraDataInterfaceSkeletalMesh
 ]
 usedDynamicInputs [&LerpLinearColors]
 usedModules [
  &SystemState
  &self:SystemSpawnScript.SystemScriptSource.SystemScriptGraph.NiagaraNodeAssignment_1.SetVariables_#1
  &self:SystemSpawnScript.SystemScriptSource.SystemScriptGraph.NiagaraNodeAssignment_0.SetVariables_#2
 ]
}
emitters {
 DirectionalBurst {
  emitterUpdateScript [
   &EmitterState()
   &SpawnBurst_Instantaneous(Use Spawn Probability=false, UseLoopCountLimit=false)
   &SpawnBurst_Instantaneous(Spawn Group=2, Spawn Time=3.7, Use Spawn Probability=false, UseLoopCountLimit=false) {module=SpawnBurst_Instantaneous001}
   &SpawnBurst_Instantaneous(Spawn Group=3, Spawn Time=3.7, Use Spawn Probability=false, UseLoopCountLimit=false) {module=SpawnBurst_Instantaneous002}
   /SetVariables(
    Emitter.GlowColorInner:LinearColor=&Multiply_LinearColorByFloat(
     Linear Color=$System.ColorCurveTargets
     Float=&FloatFromCurve(FloatCurve=DataInterface() {propertyValues {Curve {keys [<at 0 1>,<at 1 1>]}}}, CurveIndex=$Emitter.NormalizedLoopAge)
    )
    Emitter.GlowColorOuter:LinearColor=&Multiply_LinearColorByFloat(
     Linear Color=$System.ColorCurveTargets
     Float=&FloatFromCurve(FloatCurve=DataInterface() {propertyValues {Curve {keys [<at 0 1>,<at 1 1>]}}}, CurveIndex=$Emitter.NormalizedLoopAge)
    )
    Emitter.RadarScanOpacity:NiagaraFloat=&FloatFromCurve(FloatCurve=DataInterface() {propertyValues {Curve {keys [<at 0.4 1>,<at 0.5 0>]}}}, CurveIndex=$Emitter.NormalizedLoopAge)
    Emitter.TargetColorInner:LinearColor=$System.ColorCurveTargets
    Emitter.TargetColorOuter:LinearColor=$System.ColorCurveTargets
   ) {isSetParameters=true}
  ]
  particleSpawnScript [
   &FN_SetSpawnGroup() {enabled=false}
   /SetVariables(Particles.VisibilityTag=&MakeCustomIntFromBool(
     Bool=&SetBoolByIntComparison(A=&SpawnGroup(), B=0, Comparison Type=::NewEnumerator0)
     True INT=&SpawnGroup()
     False INT=/Niagara/DynamicInputs/Helpers/MaskInt(
      Mask Source:NiagaraInt32=0
      Mask Parameter:NiagaraInt32=?
      Mask:NiagaraInt32=0
      Masked Int:NiagaraInt32=&SelectIntFromArray(
       Direct Array Index=&MakeIntFromFloat(Float=$User.TargetB)
       Int Selection Array=DataInterface() {propertyValues {IntData [0,1]}}
      )
      Passthrough Int:NiagaraInt32=&SelectIntFromArray(Direct Array Index=&SpawnGroup(), Int Selection Array=DataInterface() {propertyValues {IntData [2,3,2,3]}})
     )
    )) {isSetParameters=true, module=SetVariables_#3}
   &PartitionParticles()
   &InitializeParticle(
    Color=<rgba 1 1 1 0.75>
    Lifetime=&SelectFloatFromArray(
     Direct Array Index=&SpawnGroup()
     Float Selection Array=DataInterface() {propertyValues {FloatData [7.3,7.3,0.65,0.65]}}
    )
    Mass Max=2
    Mass Mode=::NewEnumerator3
    Mesh Scale=<2 1.4 2>
    Mesh Scale Mode=::NewEnumerator4
    Sprite Rotation Angle=90
    Sprite Rotation Mode=::NewEnumerator0
    Sprite Size Mode=::NewEnumerator3
    Uniform Sprite Size=&Lerp_Float(
     A=&FN_FrontEndFloatModifier(Gameplay=75, Front End=35)
     B=&FN_FrontEndFloatModifier(Gameplay=45, Front End=22)
     Alpha=$User.TargetB
    )
    UsePositionOffset=true
    Write Material Random=true
   )
   /SetVariables(
    Particles.CrosshairInner:Vector4f=&MakeVector4(X=$User.TargetB, Y=0, Z=0, W=0)
    Particles.CrosshairOuter:Vector4f=&MakeVector4(X=$User.TargetB, Y=0, Z=0, W=0)
   ) {isSetParameters=true, module=SetVariables_#4}
   &ApplyInitialForces()
   &AlignSpriteToMeshOrientation(
    Mesh Orientation Relative Sprite Alignment Vector=<1 1 0>
    Mesh Orientation Relative Sprite Facing Vector:Vector3f=<0 1 0>
    SetAlignment=true
    SetFacing=true
   )
  ]
  particleUpdateScript [
   &ParticleState()
   &ScaleColor(
    Scale Alpha=&FloatFromCurve(FloatCurve=DataInterface() {propertyValues {Curve {keys [{arriveTangent=0.002, interpMode=RCIM_Cubic, time=0, value=1.0113}, <at 1.0011 0 RCIM_Cubic>]}}})
    ScaleA=true
    ScaleRGB=true
   )
   &SkeletalMeshLocation(
    Bone / Socket Index=1
    Bone Sampling Mode=::NewEnumerator4
    MaskSpawnGroup=false
    Position Sampling=::NewEnumerator2
    Skeletal Mesh=$System.Skeletal Mesh
   ) {enabled=false}
   &SkeletalMeshLocation(Bone Sampling Mode=::NewEnumerator4, MaskSpawnGroup=false, Skeletal Mesh=$System.Skeletal Mesh) {enabled=false, module=SkeletalMeshLocation001}
   &ScaleMeshSize(Scale Factor=&MakeVector(
     X=&FloatFromCurve(FloatCurve=DataInterface() {propertyValues {Curve {keys [
          <at 0 2 RCIM_Cubic>
          {arriveTangent=-20, interpMode=RCIM_Cubic, leaveTangent=-20, time=0.03, value=0.8}
          {arriveTangent=0.339, interpMode=RCIM_Cubic, leaveTangent=0.339, time=0.05, value=1}
          {arriveTangent=0.2037, interpMode=RCIM_Cubic, leaveTangent=0.2037, tangentMode=RCTM_User, time=0.62, value=1}
          <at 0.64 1 RCIM_Cubic>
         ]}}})
     Y=1
     Z=&FloatFromCurve(FloatCurve=DataInterface() {propertyValues {Curve {keys [
          <at 0.015 0 RCIM_Cubic>
          {arriveTangent=28.5714, interpMode=RCIM_Cubic, leaveTangent=28.5714, time=0.035, value=1.3}
          {arriveTangent=-0.5128, interpMode=RCIM_Cubic, leaveTangent=-0.5128, time=0.05, value=1}
          {arriveTangent=-0.3208, interpMode=RCIM_Cubic, leaveTangent=-0.3208, tangentMode=RCTM_User, time=0.62, value=1}
          <at 0.645 0 RCIM_Cubic>
         ]}}})
    )) {enabled=false}
   &ScaleMeshSize(Scale Factor=&MakeVector(
     X=&FloatFromCurve(FloatCurve=DataInterface() {propertyValues {Curve {keys [<at 0.64 1 RCIM_Cubic>, {arriveTangent=0.2037, interpMode=RCIM_Cubic, leaveTangent=0.2037, tangentMode=RCTM_User, time=0.68, value=0}]}}})
     Y=&FloatFromCurve(FloatCurve=DataInterface() {propertyValues {Curve {keys [{arriveTangent=0.2037, interpMode=RCIM_Cubic, leaveTangent=0.2037, tangentMode=RCTM_User, time=0.6, value=1}, <at 0.68 0 RCIM_Cubic>]}}})
     Z=&FloatFromCurve(FloatCurve=DataInterface() {propertyValues {Curve {keys [{arriveTangent=-0.3208, interpMode=RCIM_Cubic, leaveTangent=-0.3208, tangentMode=RCTM_User, time=0.7, value=1}, <at 0.73 0 RCIM_Cubic>]}}})
    )) {module=ScaleMeshSize001}
   &UpdateMeshOrientation(
    Facing Direction=&MakeVector(X=0, Y=0, Z=0)
    Facing Interpolation=false
    Orientation Method=::NewEnumerator1
    Override Quat=false
    TransformFacing=false
    TransformUp=false
   ) {enabled=false}
   /SetVariables(Particles.FrontEndAdjusted:NiagaraPosition=&Lerp_Vector(
     A=&AddVectorToPosition(Position=$Particles.Position, Vector=&MakeVector(X=0, Y=0, Z=-80))
     B=Linked() {linkedVariable {name=Particles.Position, type=NiagaraPosition}}
     Alpha=&MakeCustomFloatFromBool(
      Bool=$User.bIsFrontEndPreview
      True Float=&FloatFromCurve(FloatCurve=DataInterface() {propertyValues {Curve {keys [
           <at 0.1 0.6 RCIM_Cubic>
           {arriveTangent=0.3351, interpMode=RCIM_Cubic, leaveTangent=0.3351, tangentMode=RCTM_User, time=0.35, value=0.55}
           {arriveTangent=2.9809, interpMode=RCIM_Cubic, leaveTangent=2.9809, tangentMode=RCTM_User, time=0.45, value=1}
          ]}}})
      False Float=1
     )
    )) {isSetParameters=true}
   /SetVariables(
    Particles.InnerWiggle:Vector3f=&Multiply_Vector(
     A=&MakeVector(
      X=&Waveform(
       Waveform Count=::NewEnumerator1
       "Waveform [1]"=::NewEnumerator0
       "[1] Frequency"=1
       "[1] Amplitude Min/Max"=<-20 20>
       "Waveform [2] Blend"=::NewEnumerator5
       "[2] Interpolation"=0.4
       "[2] Invert Interpolation"=false
       "[2] Input Position"=0.3
       "[2] Phase"=$Emitter.Age
       "[2] Phase Scale"=0.5
       "[2] Compound Frequency"=2
       "[2] Frequency"=1.5
       "[2] Amplitude Min/Max"=<-10 10>
       "[2] Scale"=1
       "[2] Bias"=0
       "[2] Exponent"=1
       "[2] Pulse Duration"=0.1
       "[2] Clip Triangle Min/Max"=<0 0>
       "[2] Random Seed"=0
      )
      Y=&Waveform(
       Waveform Count=::NewEnumerator1
       Global Input Position=0.2
       "Waveform [1]"=::NewEnumerator0
       "[1] Frequency"=1
       "[1] Amplitude Min/Max"=<-20 20>
       "Waveform [2] Blend"=::NewEnumerator5
       "[2] Interpolation"=0.4
       "[2] Invert Interpolation"=false
       "[2] Input Position"=0
       "[2] Phase"=$Emitter.Age
       "[2] Phase Scale"=0.5
       "[2] Compound Frequency"=2
       "[2] Frequency"=2
       "[2] Amplitude Min/Max"=<-10 10>
       "[2] Scale"=1
       "[2] Bias"=0
       "[2] Exponent"=1
       "[2] Pulse Duration"=0.1
       "[2] Clip Triangle Min/Max"=<0 0>
       "[2] Random Seed"=0
      )
      Z=&Waveform(
       Waveform Count=::NewEnumerator1
       "Waveform [1]"=::NewEnumerator0
       "[1] Frequency"=1
       "[1] Amplitude Min/Max"=<-20 20>
       "Waveform [2] Blend"=::NewEnumerator5
       "[2] Interpolation"=0.4
       "[2] Invert Interpolation"=false
       "[2] Input Position"=0
       "[2] Phase"=$Emitter.Age
       "[2] Phase Scale"=0.5
       "[2] Compound Frequency"=2
       "[2] Frequency"=2
       "[2] Amplitude Min/Max"=<-10 10>
       "[2] Scale"=1
       "[2] Bias"=0
       "[2] Exponent"=1
       "[2] Pulse Duration"=0.1
       "[2] Clip Triangle Min/Max"=<0 0>
       "[2] Random Seed"=0
      )
     )
     B=&VectorFromFloat(Value=&FloatFromCurve(
       FloatCurve=DataInterface() {propertyValues {Curve {keys [
           {arriveTangent=0.0577, interpMode=RCIM_Cubic, leaveTangent=0.0577, tangentMode=RCTM_User, time=0, value=0.05}
           {arriveTangent=-0.1111, interpMode=RCIM_Cubic, leaveTangent=-0.1111, time=0.3, value=0.4}
           <at 0.45 0 RCIM_Cubic>
          ]}}}
       Scale Curve=0.5
      ))
    )
    Particles.OuterWiggle:Vector3f=&Multiply_Vector(
     A=&MakeVector(
      X=&Waveform(
       Waveform Count=::NewEnumerator1
       Global Input Position=0.6
       "Waveform [1]"=::NewEnumerator0
       "[1] Frequency"=0.3
       "[1] Amplitude Min/Max"=<-25 13>
       "Waveform [2] Blend"=::NewEnumerator5
       "[2] Interpolation"=0.4
       "[2] Invert Interpolation"=false
       "[2] Input Position"=0.8
       "[2] Phase"=$Emitter.Age
       "[2] Phase Scale"=0.5
       "[2] Compound Frequency"=2
       "[2] Frequency"=0.9
       "[2] Amplitude Min/Max"=<-10 10>
       "[2] Scale"=1
       "[2] Bias"=0
       "[2] Exponent"=1
       "[2] Pulse Duration"=0.1
       "[2] Clip Triangle Min/Max"=<0 0>
       "[2] Random Seed"=0
      )
      Y=&Waveform(
       Waveform Count=::NewEnumerator1
       Global Input Position=0.15
       "Waveform [1]"=::NewEnumerator0
       "[1] Frequency"=0.3
       "[1] Amplitude Min/Max"=<-12 19>
       "Waveform [2] Blend"=::NewEnumerator5
       "[2] Interpolation"=0.4
       "[2] Invert Interpolation"=false
       "[2] Input Position"=0
       "[2] Phase"=$Emitter.Age
       "[2] Phase Scale"=0.5
       "[2] Compound Frequency"=2
       "[2] Frequency"=2
       "[2] Amplitude Min/Max"=<-10 10>
       "[2] Scale"=1
       "[2] Bias"=0
       "[2] Exponent"=1
       "[2] Pulse Duration"=0.1
       "[2] Clip Triangle Min/Max"=<0 0>
       "[2] Random Seed"=0
      )
      Z=&Waveform(
       Waveform Count=::NewEnumerator1
       "Waveform [1]"=::NewEnumerator0
       "[1] Frequency"=0.7
       "[1] Amplitude Min/Max"=<-20 20>
       "Waveform [2] Blend"=::NewEnumerator5
       "[2] Interpolation"=0.4
       "[2] Invert Interpolation"=false
       "[2] Input Position"=0
       "[2] Phase"=$Emitter.Age
       "[2] Phase Scale"=0.5
       "[2] Compound Frequency"=2
       "[2] Frequency"=2
       "[2] Amplitude Min/Max"=<-10 10>
       "[2] Scale"=1
       "[2] Bias"=0
       "[2] Exponent"=1
       "[2] Pulse Duration"=0.1
       "[2] Clip Triangle Min/Max"=<0 0>
       "[2] Random Seed"=0
      )
     )
     B=&VectorFromFloat(Value=&FloatFromCurve(
       FloatCurve=DataInterface() {propertyValues {Curve {keys [
           {arriveTangent=0.0577, interpMode=RCIM_Cubic, leaveTangent=0.0577, tangentMode=RCTM_User, time=0, value=0.05}
           {arriveTangent=-0.1111, interpMode=RCIM_Cubic, leaveTangent=-0.1111, time=0.3, value=0.4}
           <at 0.45 0 RCIM_Cubic>
          ]}}}
       Scale Curve=0.5
      ))
    )
   ) {isSetParameters=true, module=SetVariables_#5}
   /SetVariables(
    Particles.InnerPos:NiagaraPosition=&AddVectorToPosition(
     Position=&AddVectorToPosition(Position=$Particles.FrontEndAdjusted, Vector=$Particles.InnerWiggle)
     Vector=&MakeVector(
      X=0
      Y=&FloatFromCurve(FloatCurve=DataInterface() {propertyValues {Curve {keys [<at 0.61 1>,<at 0.625 0>]}}}, Scale Curve=20)
      Z=0
     )
    )
    Particles.OuterPos:NiagaraPosition=&AddVectorToPosition(Position=$Particles.FrontEndAdjusted, Vector=$Particles.OuterWiggle)
   ) {isSetParameters=true, module=SetVariables_#6}
   /SetVariables(Particles.Quat:Quat4f=$Particles.SkeletalMeshLocation001.SampledRotation) {isSetParameters=true, module=SetVariables_#7}
   /SetVariables(Particles.MeshOrientation=&MakeQuaternion(
     Quaternion From=0
     Axis=<0 1 0>
     Angle=&FloatFromCurve(
      FloatCurve=DataInterface() {propertyValues {Curve {keys [
          {arriveTangent=-3529.1941, interpMode=RCIM_Cubic, leaveTangent=-3529.1941, tangentMode=RCTM_User, time=0, value=500}
          {arriveTangent=-158.2052, interpMode=RCIM_Cubic, leaveTangent=-158.2052, tangentMode=RCTM_User, time=0.45, value=0}
         ]}}}
      Scale Curve=&Lerp_Float(A=-1, B=1, Alpha=$User.TargetB)
     )
     From=?
     To=?
     Angle Type=::NewEnumerator0
     Normalize Vectors=false
    )) {isSetParameters=true}
   /SetVariables(Particles.SpriteSize=&Multiply_Vector2D(
     A=$Particles.Initial.SpriteSize
     B=&MakeVector2D(
      X=&FloatFromCurve(FloatCurve=DataInterface() {propertyValues {Curve {keys [
           <at 0 2 RCIM_Cubic>
           {arriveTangent=-22.2222, interpMode=RCIM_Cubic, leaveTangent=-22.2222, time=0.025, value=0.8}
           {arriveTangent=0.3077, interpMode=RCIM_Cubic, leaveTangent=0.3077, time=0.045, value=1}
           {arriveTangent=0.2037, interpMode=RCIM_Cubic, leaveTangent=0.2037, tangentMode=RCTM_User, time=0.675, value=1}
           <at 0.695 2 RCIM_Cubic>
          ]}}})
      Y=&FloatFromCurve(FloatCurve=DataInterface() {propertyValues {Curve {keys [
           <at 0.015 0 RCIM_Cubic>
           {arriveTangent=33.3333, interpMode=RCIM_Cubic, leaveTangent=33.3333, time=0.03, value=1.3}
           {arriveTangent=-0.4687, interpMode=RCIM_Cubic, leaveTangent=-0.4687, time=0.045, value=1}
           {arriveTangent=-0.3208, interpMode=RCIM_Cubic, leaveTangent=-0.3208, tangentMode=RCTM_User, time=0.67, value=1}
           <at 0.7 0 RCIM_Cubic>
          ]}}})
     )
    )) {enabled=false, isSetParameters=true}
   /SetVariables(Particles.SpriteSize=&Multiply_Vector2D(
     A=$Particles.Initial.SpriteSize
     B=&MakeVector2D(
      X=&FloatFromCurve(FloatCurve=DataInterface() {propertyValues {Curve {keys [
           <at 0.015 0 RCIM_Cubic>
           {arriveTangent=33.3333, interpMode=RCIM_Cubic, leaveTangent=33.3333, time=0.03, value=1.3}
           {arriveTangent=-0.5172, interpMode=RCIM_Cubic, leaveTangent=-0.5172, time=0.045, value=1}
           {arriveTangent=-0.3208, interpMode=RCIM_Cubic, leaveTangent=-0.3208, tangentMode=RCTM_User, time=0.61, value=1}
           <at 0.64 0 RCIM_Cubic>
          ]}}})
      Y=&FloatFromCurve(FloatCurve=DataInterface() {propertyValues {Curve {keys [
           <at 0 2 RCIM_Cubic>
           {arriveTangent=-22.2222, interpMode=RCIM_Cubic, leaveTangent=-22.2222, time=0.025, value=0.8}
           {arriveTangent=0.3361, interpMode=RCIM_Cubic, leaveTangent=0.3361, time=0.045, value=1}
           {arriveTangent=0.2037, interpMode=RCIM_Cubic, leaveTangent=0.2037, tangentMode=RCTM_User, time=0.62, value=1}
           <at 0.635 2 RCIM_Cubic>
          ]}}})
     )
    )) {isSetParameters=true}
   /SetVariables(Particles.SpriteRotation:NiagaraFloat=&Lerp_Float(
     A=&FloatFromCurve(
      FloatCurve=DataInterface() {propertyValues {Curve {keys [
          {arriveTangent=-0.1194, interpMode=RCIM_Cubic, leaveTangent=-0.1194, tangentMode=RCTM_User, time=0, value=1}
          {arriveTangent=-0.2268, interpMode=RCIM_Cubic, leaveTangent=-0.2268, tangentMode=RCTM_User, time=0.5, value=0}
         ]}}}
      Scale Curve=500
     )
     B=1
     Alpha=$User.TargetB
    )) {isSetParameters=true}
   /SetVariables(
    Particles.ArrowsGreen:LinearColor=&ColorFromCurve(ColorCurve=DataInterface() {propertyValues {
       AlphaCurve {keys [<at 0 0.5 RCIM_Constant>,<at 0.3183 0 RCIM_Constant>,<at 0.7276 0.5 RCIM_Constant>,<at 1 0 RCIM_Constant>]}
       BlueCurve {keys [<at 0 0.0887 RCIM_Cubic>]}
       GreenCurve {keys [<at 0 0.5913 RCIM_Cubic>]}
       RedCurve {keys [<at 0 1 RCIM_Cubic>]}
      }})
    Particles.ArrowsRed:LinearColor=&ColorFromCurve(ColorCurve=DataInterface() {propertyValues {
       AlphaCurve {keys [<at 0 0.5 RCIM_Constant>,<at 0.1955 0.5 RCIM_Constant>,<at 0.3019 0 RCIM_Constant>,<at 0.5986 0 RCIM_Constant>,<at 0.7051 0.5 RCIM_Constant>,<at 1 0 RCIM_Constant>]}
       BlueCurve {keys [<at 0 0.1248 RCIM_Cubic>]}
       GreenCurve {keys [<at 0 0.6584 RCIM_Cubic>]}
       RedCurve {keys [<at 0 1 RCIM_Cubic>]}
      }})
   ) {isSetParameters=true}
   &DynamicMaterialParameters(
    Index 0 Param 2=2
    Index 0 Param 3=0.05
    Index 3 Param 1=&FloatFromCurve(FloatCurve=DataInterface() {propertyValues {Curve {keys [<at 0.01 -0.5>,<at 0.1 1>]}}})
    Module.Index0Float0Write=true
    Module.Index0Float1Write=true
    Module.Index0Float2Write=true
    Module.Index0Float3Write=true
    Module.Index3Float0Write=true
    Module.Index3Float1Write=true
    Module.Index3Float2Write=true
    Module.Index3Float3Write=true
    Write Parameter Index 3=::NewEnumerator1
   )
   /SetVariables(Particles.Arrows1:Vector4f=<1 1 0 1>, Particles.Arrows2:Vector4f=<1 0 0 0>) {isSetParameters=true}
   /SetVariables(
    Particles.GreenCenter:Vector2f=&Multiply_Vector2D_ByFloat(
     Float=&FloatFromCurve(FloatCurve=DataInterface() {propertyValues {Curve {defaultValue=1, keys [{tangentMode=RCTM_User, time=0, value=0}, {tangentMode=RCTM_User, time=0.02, value=1}, <at 0.9 1>, <at 1 0>]}}})
     Vector2D=<15 15>
    )
    Particles.GreenRotation:NiagaraFloat=&FloatFromCurve(FloatCurve=DataInterface() {propertyValues {Curve {keys [<at 0 1>,<at 1 0>]}}}, Scale Curve=-50)
    Particles.RedCenter:Vector2f=&Multiply_Vector2D_ByFloat(
     Float=&FloatFromCurve(FloatCurve=DataInterface() {propertyValues {Curve {defaultValue=1, keys [{tangentMode=RCTM_User, time=0, value=0}, {tangentMode=RCTM_User, time=0.02, value=1}, <at 0.9 1>, <at 1 0>]}}})
     Vector2D=<15 15>
    )
    Particles.RedRotation:NiagaraFloat=&FloatFromCurve(FloatCurve=DataInterface() {propertyValues {Curve {keys [<at 0 1>,<at 1 0>]}}}, Scale Curve=50)
   ) {isSetParameters=true}
   /SetVariables(
    Particles.Arrow1Pos:NiagaraPosition=&AddVectorToPosition(Position=$Particles.Position, Vector=<0 10 0>)
    Particles.Arrow2Pos:NiagaraPosition=&AddVectorToPosition(Position=$Particles.Position, Vector=<0 5 0>)
   ) {isSetParameters=true}
  ]
  rendererClasses [&NiagaraMeshRendererProperties, &NiagaraSpriteRendererProperties]
  renderers [
   {rendererClass=&NiagaraMeshRendererProperties, rendererIndex=0}
   {rendererClass=&NiagaraMeshRendererProperties, rendererIndex=1}
   {rendererClass=&NiagaraSpriteRendererProperties, rendererIndex=2}
   {rendererClass=&NiagaraSpriteRendererProperties, rendererIndex=3}
   {rendererClass=&NiagaraSpriteRendererProperties, rendererIndex=4}
   {rendererClass=&NiagaraSpriteRendererProperties, rendererIndex=5}
  ]
 }
 DirectionalBurst001 {
  emitterUpdateScript [
   &EmitterState()
   &SpawnBurst_Instantaneous(Spawn Time=3.7, Use Spawn Probability=false, UseLoopCountLimit=false)
   /SetVariables(
    Emitter.GlowColorInner:LinearColor=$System.ColorCurveTargets
    Emitter.GlowColorOuter:LinearColor=$System.ColorCurveTargets
    Emitter.TargetColorInner:LinearColor=$System.ColorCurveTargets
    Emitter.TargetColorOuter:LinearColor=$System.ColorCurveTargets
   ) {isSetParameters=true}
  ]
  particleSpawnScript [
   &FN_SetSpawnGroup(Number of Spawn Groups=2)
   &InitializeParticle(
    Lifetime=0.25
    Mass Max=2
    Mass Mode=::NewEnumerator3
    Mesh Scale Mode=::NewEnumerator3
    Mesh Uniform Scale=2
    Sprite Size Max=&FN_FrontEndVector2Modifier(Front End:Vector2f=<1.5 7>, Gameplay:Vector2f=<3 14>)
    Sprite Size Min=&FN_FrontEndVector2Modifier(Front End:Vector2f=<1 3>, Gameplay:Vector2f=<2 6>)
    Sprite Size Mode=::NewEnumerator2
    UsePositionOffset=true
    Write Material Random=true
   )
   &PartitionParticles()
   /SetVariables(
    Particles.CrosshairInner:Vector4f=&MakeVector4(X=0, Y=0, Z=0, W=0)
    Particles.CrosshairOuter:Vector4f=&MakeVector4(X=0, Y=0, Z=0, W=0)
   ) {isSetParameters=true, module=SetVariables_#4}
   &DynamicMaterialParameters(
    Index 3 Param 1=1
    Module.Index3Float0Write=true
    Module.Index3Float1Write=true
    Module.Index3Float2Write=true
    Module.Index3Float3Write=true
    Write Parameter Index 0=::NewEnumerator0
    Write Parameter Index 3=::NewEnumerator1
   )
   /SetVariables(Particles.VisibilityTag=&SelectIntFromArray(
     Direct Array Index=&MakeIntFromFloat(Float=$User.TargetB)
     Int Selection Array=DataInterface() {propertyValues {IntData [0,1]}}
    )) {isSetParameters=true, module=SetVariables_#3}
  ]
  particleUpdateScript [
   &ParticleState()
   &ScaleColor(
    Scale Alpha=&FloatFromCurve(FloatCurve=DataInterface() {propertyValues {Curve {keys [{arriveTangent=-0.9385, interpMode=RCIM_Cubic, leaveTangent=-0.9385, tangentMode=RCTM_User, time=0, value=0.3}, <at 1 0 RCIM_Cubic>]}}})
    ScaleA=true
    ScaleRGB=true
   )
   &SkeletalMeshLocation(
    Bone / Socket Index=1
    Bone Sampling Mode=::NewEnumerator4
    MaskSpawnGroup=false
    Position Sampling=::NewEnumerator2
    Skeletal Mesh=$System.Skeletal Mesh
   ) {enabled=false}
   &SkeletalMeshLocation(Bone Sampling Mode=::NewEnumerator4, MaskSpawnGroup=false, Skeletal Mesh=$System.Skeletal Mesh) {enabled=false, module=SkeletalMeshLocation001}
   &ScaleMeshSize(Scale Factor=&VectorFromFloat(Value=&FloatFromCurve(
      FloatCurve=DataInterface() {propertyValues {Curve {keys [
          {arriveTangent=4.842, interpMode=RCIM_Cubic, leaveTangent=4.842, tangentMode=RCTM_User, time=0, value=1}
          {arriveTangent=0.5631, interpMode=RCIM_Cubic, leaveTangent=0.5631, tangentMode=RCTM_User, time=0.1452, value=1.3786}
          <at 1 1.5 RCIM_Cubic>
         ]}}}
      Scale Curve=&FN_FrontEndFloatModifier(Gameplay=1, Front End=0.5)
     )))
   &UpdateMeshOrientation(
    Facing Direction=&MakeVector(X=0, Y=0, Z=0)
    Facing Interpolation=false
    Orientation Method=::NewEnumerator1
    Override Quat=false
    TransformFacing=false
    TransformUp=false
   ) {enabled=false}
   /SetVariables(
    Particles.InnerWiggle:Vector3f=&Multiply_Vector(
     A=&MakeVector(
      X=&Waveform(
       Waveform Count=::NewEnumerator1
       "Waveform [1]"=::NewEnumerator0
       "[1] Frequency"=1
       "[1] Amplitude Min/Max"=<-20 20>
       "Waveform [2] Blend"=::NewEnumerator5
       "[2] Interpolation"=0.4
       "[2] Invert Interpolation"=false
       "[2] Input Position"=0.3
       "[2] Phase"=$Emitter.Age
       "[2] Phase Scale"=0.5
       "[2] Compound Frequency"=2
       "[2] Frequency"=1.5
       "[2] Amplitude Min/Max"=<-10 10>
       "[2] Scale"=1
       "[2] Bias"=0
       "[2] Exponent"=1
       "[2] Pulse Duration"=0.1
       "[2] Clip Triangle Min/Max"=<0 0>
       "[2] Random Seed"=0
      )
      Y=&Waveform(
       Waveform Count=::NewEnumerator1
       Global Input Position=0.2
       "Waveform [1]"=::NewEnumerator0
       "[1] Frequency"=1
       "[1] Amplitude Min/Max"=<-20 20>
       "Waveform [2] Blend"=::NewEnumerator5
       "[2] Interpolation"=0.4
       "[2] Invert Interpolation"=false
       "[2] Input Position"=0
       "[2] Phase"=$Emitter.Age
       "[2] Phase Scale"=0.5
       "[2] Compound Frequency"=2
       "[2] Frequency"=2
       "[2] Amplitude Min/Max"=<-10 10>
       "[2] Scale"=1
       "[2] Bias"=0
       "[2] Exponent"=1
       "[2] Pulse Duration"=0.1
       "[2] Clip Triangle Min/Max"=<0 0>
       "[2] Random Seed"=0
      )
      Z=&Waveform(
       Waveform Count=::NewEnumerator1
       "Waveform [1]"=::NewEnumerator0
       "[1] Frequency"=1
       "[1] Amplitude Min/Max"=<-20 20>
       "Waveform [2] Blend"=::NewEnumerator5
       "[2] Interpolation"=0.4
       "[2] Invert Interpolation"=false
       "[2] Input Position"=0
       "[2] Phase"=$Emitter.Age
       "[2] Phase Scale"=0.5
       "[2] Compound Frequency"=2
       "[2] Frequency"=2
       "[2] Amplitude Min/Max"=<-10 10>
       "[2] Scale"=1
       "[2] Bias"=0
       "[2] Exponent"=1
       "[2] Pulse Duration"=0.1
       "[2] Clip Triangle Min/Max"=<0 0>
       "[2] Random Seed"=0
      )
     )
     B=&VectorFromFloat(Value=&FloatFromCurve(FloatCurve=DataInterface() {propertyValues {Curve {keys [
           {arriveTangent=0.0577, interpMode=RCIM_Cubic, leaveTangent=0.0577, tangentMode=RCTM_User, time=0, value=0.05}
           {arriveTangent=-0.1111, interpMode=RCIM_Cubic, leaveTangent=-0.1111, time=0.3, value=0.4}
           <at 0.45 0 RCIM_Cubic>
          ]}}}))
    )
    Particles.OuterWiggle:Vector3f=&Multiply_Vector(
     A=&MakeVector(
      X=&Waveform(
       Waveform Count=::NewEnumerator1
       Global Input Position=0.6
       "Waveform [1]"=::NewEnumerator0
       "[1] Frequency"=0.3
       "[1] Amplitude Min/Max"=<-25 13>
       "Waveform [2] Blend"=::NewEnumerator5
       "[2] Interpolation"=0.4
       "[2] Invert Interpolation"=false
       "[2] Input Position"=0.8
       "[2] Phase"=$Emitter.Age
       "[2] Phase Scale"=0.5
       "[2] Compound Frequency"=2
       "[2] Frequency"=0.9
       "[2] Amplitude Min/Max"=<-10 10>
       "[2] Scale"=1
       "[2] Bias"=0
       "[2] Exponent"=1
       "[2] Pulse Duration"=0.1
       "[2] Clip Triangle Min/Max"=<0 0>
       "[2] Random Seed"=0
      )
      Y=&Waveform(
       Waveform Count=::NewEnumerator1
       Global Input Position=0.15
       "Waveform [1]"=::NewEnumerator0
       "[1] Frequency"=0.3
       "[1] Amplitude Min/Max"=<-12 19>
       "Waveform [2] Blend"=::NewEnumerator5
       "[2] Interpolation"=0.4
       "[2] Invert Interpolation"=false
       "[2] Input Position"=0
       "[2] Phase"=$Emitter.Age
       "[2] Phase Scale"=0.5
       "[2] Compound Frequency"=2
       "[2] Frequency"=2
       "[2] Amplitude Min/Max"=<-10 10>
       "[2] Scale"=1
       "[2] Bias"=0
       "[2] Exponent"=1
       "[2] Pulse Duration"=0.1
       "[2] Clip Triangle Min/Max"=<0 0>
       "[2] Random Seed"=0
      )
      Z=&Waveform(
       Waveform Count=::NewEnumerator1
       "Waveform [1]"=::NewEnumerator0
       "[1] Frequency"=0.7
       "[1] Amplitude Min/Max"=<-20 20>
       "Waveform [2] Blend"=::NewEnumerator5
       "[2] Interpolation"=0.4
       "[2] Invert Interpolation"=false
       "[2] Input Position"=0
       "[2] Phase"=$Emitter.Age
       "[2] Phase Scale"=0.5
       "[2] Compound Frequency"=2
       "[2] Frequency"=2
       "[2] Amplitude Min/Max"=<-10 10>
       "[2] Scale"=1
       "[2] Bias"=0
       "[2] Exponent"=1
       "[2] Pulse Duration"=0.1
       "[2] Clip Triangle Min/Max"=<0 0>
       "[2] Random Seed"=0
      )
     )
     B=&VectorFromFloat(Value=&FloatFromCurve(FloatCurve=DataInterface() {propertyValues {Curve {keys [
           {arriveTangent=0.0577, interpMode=RCIM_Cubic, leaveTangent=0.0577, tangentMode=RCTM_User, time=0, value=0.05}
           {arriveTangent=-0.1111, interpMode=RCIM_Cubic, leaveTangent=-0.1111, time=0.3, value=0.4}
           <at 0.45 0 RCIM_Cubic>
          ]}}}))
    )
   ) {isSetParameters=true, module=SetVariables_#5}
   /SetVariables(
    Particles.InnerPos:NiagaraPosition=&AddVectorToPosition(Position=$Particles.SkeletalMeshLocation.SampledPosition, Vector=$Particles.InnerWiggle)
    Particles.OuterPos:NiagaraPosition=&AddVectorToPosition(Position=$Particles.SkeletalMeshLocation001.SampledPosition, Vector=$Particles.OuterWiggle)
   ) {isSetParameters=true, module=SetVariables_#6}
   /SetVariables(Particles.Quat:Quat4f=$Particles.SkeletalMeshLocation001.SampledRotation) {isSetParameters=true, module=SetVariables_#7}
  ]
  rendererClasses [&NiagaraMeshRendererProperties]
  renderers [
   {rendererClass=&NiagaraMeshRendererProperties, rendererIndex=0}
   {rendererClass=&NiagaraMeshRendererProperties, rendererIndex=1}
  ]
 }
 DirectionalBurst002 {
  emitterUpdateScript [
   &EmitterState(Life Cycle Mode=::NewEnumerator1, Loop Behavior=::NewEnumerator1, Loop Duration=6, Scalability Mode=::NewEnumerator1, UseLoopDelay=false)
   &SpawnBurst_Instantaneous(Spawn Time=4.5, Use Spawn Probability=false, UseLoopCountLimit=false) {module=SpawnBurst_Instantaneous001}
  ]
  particleSpawnScript [
   &InitializeParticle(
    Color=<rgba 1.2 0.79 0.1497 1>
    Lifetime=0.4
    Mass Max=2
    Mass Mode=::NewEnumerator3
    Sprite Rotation Angle=35
    Sprite Rotation Mode=::NewEnumerator0
    Sprite Size Mode=::NewEnumerator3
    Uniform Sprite Size=&RandomRangeFloat(
     Minimum=&FN_FrontEndFloatModifier(Gameplay=115, Front End=50)
     Maximum=&FN_FrontEndFloatModifier(Gameplay=115, Front End=50)
    )
    UsePositionOffset=true
    Write Material Random=true
   )
   &ApplyInitialForces()
   &AlignSpriteToMeshOrientation(Mesh Orientation Relative Sprite Facing Vector:Vector3f=<0 -1 0>, SetAlignment=true, SetFacing=true)
  ]
  particleUpdateScript [
   &ParticleState()
   &ScaleSpriteSize(Uniform Curve Sprite Scale=DataInterface() {propertyValues {Curve {
       defaultValue=0
       keys [{arriveTangent=-1.7678, interpMode=RCIM_Cubic, leaveTangent=-1.7678, tangentMode=RCTM_User, time=0, value=1}, <at 1 0 RCIM_Cubic>]
      }}})
   &ScaleColor(
    Scale Alpha=&FloatFromCurve(FloatCurve=DataInterface() {propertyValues {Curve {keys [<at -0.0046 1.0113 RCIM_Cubic>,<at 1.0011 0.002 RCIM_Cubic>]}}})
    ScaleA=true
    ScaleRGB=true
   )
   &SolveForcesAndVelocity(Clamp Velocity=false, Limit Acceleration=false)
   /SetVariables(Particles.SpriteRotation:NiagaraFloat=&FloatFromCurve(
     FloatCurve=DataInterface() {propertyValues {Curve {keys [
         {arriveTangent=-2.876, interpMode=RCIM_Cubic, leaveTangent=-2.876, tangentMode=RCTM_User, time=0, value=1}
         {arriveTangent=-0.3145, interpMode=RCIM_Cubic, leaveTangent=-0.3145, tangentMode=RCTM_User, time=1, value=0}
        ]}}}
     Scale Curve=35
    )) {isSetParameters=true}
  ]
  rendererClasses [&NiagaraSpriteRendererProperties]
  renderers [{rendererClass=&NiagaraSpriteRendererProperties}]
 }
 Flash001 {
  emitterUpdateScript [
   &EmitterState()
   &SpawnBurst_Instantaneous(Spawn Count=4, Spawn Time=4.5, Use Spawn Probability=false)
  ]
  particleSpawnScript [
   &PartitionParticles(Number Of Partitions=4)
   &InitializeParticle(
    Color=<rgba 1 0.4179 0 1>
    Lifetime=0.25
    Sprite Size Mode=::NewEnumerator3
    Uniform Sprite Size=&Multiply_Float(
     A=&FN_FrontEndFloatModifier(Gameplay=30, Front End=10)
     B=&FN_FrontEndFloatModifier(Gameplay=1, Front End=0.5)
    )
    UsePositionOffset=true
    Write Material Random=true
   )
   &ShapeLocation(MaskSpawnGroup=false, Override Seed=false, Sphere Radius=0, UseOwnerScale=true)
   &CameraOffset(Camera Offset Amount=-10)
   /SetVariables(Particles.SizeWideFlare:Vector2f=&Multiply_Vector2D_ByFloat(Float=&Multiply_Float(A=1, B=1), Vector2D=<225 40>)) {isSetParameters=true}
   &InitialMeshOrientation(Mesh Orientation Mode=::NewEnumerator2, Orientation Vector=<0 1 0>, Rotation=<0.25 0 0>, Use Rotation Vector=true)
   &AlignSpriteToMeshOrientation(SetAlignment=true, SetFacing=true)
  ]
  particleUpdateScript [
   &ParticleState()
   &SolveForcesAndVelocity(Clamp Velocity=false, Limit Acceleration=false)
   &ScaleColor(
    Scale Alpha=&FloatFromCurve(FloatCurve=DataInterface() {propertyValues {Curve {keys [
         <at 0.1 0 RCIM_Cubic>
         {interpMode=RCIM_Cubic, leaveTangent=-2.0997, tangentMode=RCTM_Break, time=0.2521, value=1}
         {arriveTangentWeight=0.7369, interpMode=RCIM_Cubic, tangentMode=RCTM_Break, time=1, value=1}
        ]}}})
    Scale RGB=&VectorFromFloat(Value=&FloatFromCurve(
      FloatCurve=DataInterface() {propertyValues {Curve {keys [
          <at 0.1 0 RCIM_Cubic>
          {interpMode=RCIM_Cubic, leaveTangent=-2.0997, tangentMode=RCTM_Break, time=0.2521, value=1}
          {arriveTangentWeight=0.7369, interpMode=RCIM_Cubic, tangentMode=RCTM_Break, time=0.6, value=1}
         ]}}}
      Scale Curve=5
     ))
    ScaleA=true
    ScaleRGB=true
   )
   &ScaleSpriteSize(Uniform Curve Sprite Scale=DataInterface() {propertyValues {Curve {keys [
        <at 0.3 0 RCIM_Cubic>
        {interpMode=RCIM_Cubic, tangentMode=RCTM_User, time=0.4, value=1}
        {interpMode=RCIM_Cubic, tangentMode=RCTM_User, time=1, value=0}
       ]}}})
   &FN_Sprite_Attributes(
    Color=&MakeLinearColorFromVectorAndFloat(
     "Vector (RGB)"=&MakeVectorFromLinearColorRGB(Color=<rgba 3 1.2538 0 1>)
     "Float (Alpha)"=&Multiply_Float(
      A=1
      B=&FloatFromCurve(FloatCurve=DataInterface() {propertyValues {Curve {keys [{interpMode=RCIM_Cubic, tangentMode=RCTM_User, time=0, value=1}, {interpMode=RCIM_Cubic, tangentMode=RCTM_User, time=1, value=0.5}]}}})
     )
    )
    Create Color Attribute=true
    Write to Module Namespace=true
   )
   /SetVariables(Particles.SizeWideFlare:Vector2f=&Multiply_Vector2D(
     A=$Particles.Initial.SizeWideFlare
     B=&MakeVector2D(
      X=&FloatFromCurve(FloatCurve=DataInterface() {propertyValues {Curve {keys [<at 0.1 0>,<at 0.2 0.4>,<at 0.3 1>,<at 1 0.1>]}}})
      Y=&FloatFromCurve(FloatCurve=DataInterface() {propertyValues {Curve {keys [<at 0.2 1>,<at 0.5 0.1>]}}})
     )
    )) {isSetParameters=true}
  ]
  rendererClasses [&NiagaraSpriteRendererProperties, &NiagaraMeshRendererProperties]
  renderers [
   {rendererClass=&NiagaraSpriteRendererProperties}
   {rendererClass=&NiagaraSpriteRendererProperties, rendererIndex=1}
   {rendererClass=&NiagaraMeshRendererProperties, rendererIndex=2}
  ]
 }
}
summary {
 systemName=Juno_NS_Emote_AshenMagnus
 emitters [
  {
   emitterName=DirectionalBurst
   rendererClasses [&NiagaraMeshRendererProperties, &NiagaraSpriteRendererProperties]
  }
  {emitterName=DirectionalBurst001, rendererClasses [&NiagaraMeshRendererProperties]}
  {emitterName=Flash001, rendererClasses [&NiagaraSpriteRendererProperties, &NiagaraMeshRendererProperties]}
  {emitterName=DirectionalBurst002, rendererClasses [&NiagaraSpriteRendererProperties]}
 ]
 userVariables [
  User.bIsFrontEnd {type=NiagaraBool}
  User.bIsFrontEndPreview {type=NiagaraBool}
  User.ParaGlideLeanAlpha {type=NiagaraFloat, defaultValue=0, description=""}
  User.TargetB {type=NiagaraFloat, defaultValue=1, description=""}
  User.TargetLock {type=NiagaraFloat, defaultValue=0, description=""}
  User.TargetSelect {type=NiagaraFloat, defaultValue=0, description=""}
 ]
}
