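# Translucent unlit two-sided material for Niagara mesh particles that draws the glowing holographic crosshairs in the AshenMagnus emote.
# World-height scanlines, scaled by camera distance, tint the vertex-red regions, while a CrosshairSelect-shifted radial gradient glows on vertex-blue.
# Named reroutes build time-driven flicker noise, glitch-distorted UVs, a screen-space pixel-block mask and a local-X reveal mask with triplanar noise.
# Niagara dynamic parameters and particle color control the effect.
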Material /BRCosmetics/Animation/Game/MainPlayer/Emotes/AshenMagnus/FX/M_Emote_AshenMagnus_Crosshairs
graph {
 properties {
  bAutomaticallySetUsageInEditor=false
  bUsedWithNiagaraMeshParticles=true
  blendMode=BLEND_Translucent
  shadingModel=MSM_Unlit
  twoSided=true
  usedShadingModels=Unlit
 }
 nodes [
  n1 = ScalarParameter() {defaultValue=0.4, parameterName=MinScanlineOpacity}
  n2 = Multiply(A=NamedRerouteUsage()) {constB=1.5}
  n3 = Saturate(None=NamedRerouteUsage())
  n4 = Multiply(
   A=Multiply(
    A=DynamicParameter() {defaultValue=<rgba 1 1 0.1 0.1>, paramNames [Mask,Scanline,Flicker,Random]}.Scanline
    B=ScalarParameter() {defaultValue=300, parameterName=ScanlineTile}
   )
   B=LinearInterpolate(Alpha=Ceil(None=@n3)) {constA=1, constB=0.7}
  )
  n5 = ScalarParameter() {defaultValue=20000, parameterName=ScaleDistance}
  n6 = Divide(A=Distance(A=CameraPositionWS(), B=ActorPositionWS()), B=@n5) {constB=1000}
  n7 = Divide(A=ComponentMask(None=WorldPosition().XYZ) {b=true}, B=LinearInterpolate(B=@n5, Alpha=Saturate(None=@n6)) {constA=50, constB=100}) {constB=20}
  n8 = ConstantBiasScale(None=Sine(None=Multiply(A=@n4, B=@n7) {constA=1}))
  n9 = VertexNormalWS()
  n10 = Constant3Vector() {constant=<rgba 0 0 1 0>}
  n11 = Power(Base=Abs(None=DotProduct(A=@n9, B=@n10))) {constExponent=5}
  n12 = LinearInterpolate(
   A=Clamp(None=Subtract(A=@n1, B=@n2)) {minDefault=-0.25}
   B=Add(A=@n2)
   Alpha=Power(Base=LinearInterpolate(A=@n8, Alpha=@n11) {constB=0.25}) {constExponent=1.5}
  ) {constA=0.5}
  n13 = NamedRerouteUsage()
  n14 = ParticleColor()
  n15 = Multiply(A=Multiply(A=VectorParameter() {parameterName=TargetColor, defaultValue=<rgba 1 0.0546 0 1>}.RGB, B=Multiply(A=@n12, B=@n13)), B=@n14.RGB)
  n16 = VertexColor()
  n17 = DynamicParameter() {parameterIndex=2, defaultValue=<rgba 1 1 0 1>, paramNames [CrosshairSelect, "", "", ""]}
  n18 = Saturate(None=LinearInterpolate(A=@n16.R, B=@n16.R, Alpha=@n17.CrosshairSelect))
  n19 = TextureCoordinate() {uTiling=0.1, vTiling=4.5}
  n20 = AppendVector(A=ScalarParameter(), B=LinearInterpolate(Alpha=@n17.CrosshairSelect) {constA=-2.8, constB=-0.7})
  n21 = Multiply(
   A=@n16.B
   B=MaterialFunctionCall(UVs=Add(A=@n19, B=@n20)) {materialFunction=&RadialGradientExponential}.RadialGradientExponential
  )
  n22 = Saturate(None=@n21)
  n23 = Multiply(A=VectorParameter() {parameterName=GlowColor, defaultValue=<rgba 10 0.8501 0.4994 1>}.RGB, B=@n16.B)
  MaterialFunctionCall(Emissive=Add(A=Multiply(A=@n15, B=@n18), B=Multiply(A=@n22, B=@n23))) {materialFunction=&MF_FX_EmissiveAdjustments}
  n24 = TextureObject() {texture=/BRCosmetics/Animation/Game/MainPlayer/Emotes/JadeTowel/FX/T_Emote_JadeTowel_PixelNoise}
  MaterialFunctionCall!(CameraVector, Texture=@n24, Tiling, Offset, AxisFadeContrast) {materialFunction=/Engine/Functions/Engine_MaterialFunctions01/Texturing/TriplanarCameraVector}
  n25 = Add(A=@n21, B=Saturate(None=Multiply(A=@n12, B=@n18)))
  Multiply(A=Multiply(A=Saturate(None=@n25), B=@n14.A), B=NamedRerouteUsage())
  n26 = TextureCoordinate()
  n27 = ScalarParameter() {defaultValue=0.4, parameterName=XAdd}
  n28 = ScalarParameter() {defaultValue=0.4, parameterName=YAdd}
  Multiply(
   A=Add(A=ComponentMask(None=@n26) {g=true}, B=AppendVector(A=@n27, B=@n28))
   B=AppendVector(A=ScalarParameter() {defaultValue=1, parameterName=XMult}, B=ScalarParameter() {defaultValue=0.4, parameterName=YMult})
  ) {constB=20}
  n29 = VectorParameter() {group=Noise, parameterName=NoiseTileMult, defaultValue=<rgba 1.5 1.5 0.06 1>}
  n30 = DynamicParameter() {parameterIndex=1, defaultValue=<rgba 1 1 0 1>, paramNames [NoiseTileMult, Param2, Param3, Param4]}
  n31 = Divide(A=ObjectRadius(), B=Multiply(A=@n30.NoiseTileMult, B=ScalarParameter() {defaultValue=0.25, group=Noise, parameterName=NoiseSize}))
  n32 = WorldPosition()
  n33 = Add(
   A=Add(A=DynamicParameter() {defaultValue=<rgba 1 1 0.1 0.1>, paramNames [Mask,Scanline,Flicker,Random]}.Flicker, B=Time())
   B=NamedRerouteUsage()
  )
  n34 = Multiply(A=@n31, B=Multiply(A=@n33, B=VectorParameter() {group=Noise, parameterName=NoisePanSpeed, defaultValue=<rgba -0.08 0.07 0.15 0>}.RGB))
  n35 = Power(Base=MaterialFunctionCall(
    TextureObject=TextureObject() {samplerType=SAMPLERTYPE_LinearColor, texture=&T_NoisePack_LinearColor}
    TextureSize=Multiply(A=@n29.RGB, B=@n31)
    WorldPosition=Add(A=@n32.XYZ, B=@n34)
   ) {materialFunction=&WorldAlignedTexture}.XY Texture) {constExponent=4}
  n36 = Multiply(A=@n35, B=ScalarParameter() {defaultValue=2, group=Noise, parameterName=NoiseMaskMult}) {constB=20}
  n37 = DynamicParameter() {defaultValue=<rgba 1 1 0.1 0.1>, paramNames [Mask,Scanline,Flicker,Random]}
  n38 = Multiply(A=ScalarParameter() {defaultValue=2, parameterName=FlickerSpeed}, B=@n33)
  n39 = AppendVector(A=AppendVector(A=Sine(None=@n38) {period=1.6}, B=Sine(None=@n38) {period=1.42}), B=Sine(None=@n38) {period=1.51})
  n40 = Ceil(None=@n39)
  n41 = ScalarParameter() {defaultValue=0.5, group=Noise, parameterName=NoiseFrequency}
  n42 = Add(A=Sine(None=@n38) {period=0.72}, B=Multiply(B=@n41) {constA=0.75}) {constB=0.75}
  n43 = Multiply(A=Add(A=Sine(None=@n38) {period=2.1}, B=Multiply(B=@n41) {constA=0.25}), B=@n42)
  n44 = Multiply(B=@n41) {constA=0.9, constB=0.9}
  n45 = Sine(None=@n38) {period=1.25}
  n46 = Multiply(A=@n42, B=Ceil(None=Add(A=@n44, B=@n45) {constB=0.9}))
  n47 = Sine(None=@n33) {period=0.075}
  n48 = Multiply(A=Saturate(None=Multiply(A=@n43, B=@n46)), B=LinearInterpolate(Alpha=ConstantBiasScale(None=@n47)) {constA=0.7})
  n49 = Add(A=@n37.Flicker, B=Multiply(A=Multiply(A=@n40, B=@n48), B=ScalarParameter() {defaultValue=2, group=Noise, parameterName=NoiseMult}))
  n50 = Desaturation(None=Saturate(None=Multiply(A=@n36, B=@n49)))
  n51 = ScalarParameter() {defaultValue=0.2, group=Noise, parameterName=NoiseContrast}
  NamedRerouteDeclaration(None=QualitySwitch(
    Default=MaterialFunctionCall(In=@n50, Contrast=@n51) {materialFunction=&CheapContrast}.Result
    Low=Constant()
   )) {name=FlickerNoise, nodeColor=<rgba 0.0312 0 1 1>}
  NamedRerouteDeclaration(None=@n37.Random) {name=Random, nodeColor=<rgba 0.7578 0 1 1>}
  n52 = ComponentMask(None=LocalPosition().XYZ) {r=true}
  n53 = LinearInterpolate(A=OneMinus(None=@n52), B=@n52, Alpha=ScalarParameter())
  n54 = Saturate(None=Add(
    A=Divide(A=@n53) {constB=150}
    B=DynamicParameter() {parameterIndex=3, defaultValue=<rgba 0 1 0.1 0.1>, paramNames [Mask,Scanline,Flicker,Random]}.Mask
   ))
  n55 = ComponentMask(None=MaterialFunctionCall!(
    TextureObject=@n24
    TextureSize=Constant() {r=500}
    WorldPosition
    Export Float 4
    Parallax
    In
    World Space Normal
    ProjectionTransitionContrast
   ) {materialFunction=/BRCosmetics/Characters/Player/Male/Medium/Bodies/M_MED_SilverBellMarine/Materials/MF_Triplanar_ScalePan_Parallax}.XYZ Texture) {r=true}
  n56 = Clamp(None=Subtract(A=@n54, B=SmoothStep(Value=@n55)))
  NamedRerouteDeclaration(None=Saturate(None=Add(A=@n54, B=@n56))) {name=Mask, nodeColor=<rgba 0.9688 0 1 1>}
  NamedRerouteDeclaration(None=Multiply(A=@n35, B=Add(A=@n39))) {name=NoiseDistortion, nodeColor=<rgba 0.9219 0 1 1>}
  n57 = ScreenPosition()
  n58 = Add(
   A=ScalarParameter() {defaultValue=4, group=Block, parameterName=Block UV}
   B=DynamicParameter() {defaultValue=<rgba 1 1 0.1 0.1>, paramNames [Mask,Scanline,Flicker,Random]}.Mask
  )
  n59 = Multiply(A=Time(), B=ScalarParameter() {defaultValue=5, group=Block, parameterName=Block U Speed}) {constB=0.5}
  n60 = MaterialFunctionCall(Float3=Saturate(None=NamedRerouteUsage())) {materialFunction=&BreakOutFloat3Components}
  n61 = Subtract(A=@n60.R, B=@n60.R)
  n62 = Multiply(A=@n61, B=@n61)
  n63 = Constant()
  n64 = Ceil(None=Saturate(None=NamedRerouteUsage()))
  n65 = DynamicParameter() {defaultValue=<rgba 1 1 0.1 0.1>, paramNames [Mask,Scanline,Flicker,Random]}
  n66 = Multiply(
   A=Multiply(A=AppendVector(A=@n62, B=@n63), B=Add(A=@n64, B=@n65.Flicker))
   B=ScalarParameter() {defaultValue=0.15, parameterName=NoiseDistortionAmount}
  )
  n67 = Multiply(A=Time(), B=ScalarParameter() {parameterName=Rotator Speed})
  n68 = Multiply(
   A=@n67
   B=TextureSample(UVs=@n67) {samplerType=SAMPLERTYPE_Grayscale, texture=/Game/Effects/Fort_Effects/Textures/Fire/T_WispyFireNoise_Soft_M}.R
  )
  n69 = StaticSwitch(True=Rotator(Time=Subtract(A=@n68) {constB=-1}), False=TextureCoordinate(), Value=StaticBoolParameter() {parameterName=IsRotate})
  n70 = Add(
   A=Panner(Coordinate=Multiply(A=@n57.ViewportUV, B=@n58), Time=Ceil(None=@n59)) {desc="-0.33,-0.43", speedX=-0.33, speedY=-0.43}
   B=NamedRerouteDeclaration(None=Add(A=@n66, B=@n69)) {name=Glitch UVs, nodeColor=<rgba 0 1 0.9531 1>}
  )
  NamedRerouteDeclaration(None=LinearInterpolate(
    A=ScalarParameter() {defaultValue=0.4, parameterName=PixelMaskOpacity}
    Alpha=TextureSample(UVs=@n70) {texture=/Game/Athena/UI/Challenges/Art/T_UI_DataChip_Noise}.G
   ) {constA=0.4}) {name=Main Mask, nodeColor=<rgba 0 0.6484 1 1>}
  ScalarParameter() {defaultValue=-0.7, parameterName=Param2}
  TextureSample() {texture=&T_Emote_JadeTowel_PixelNoise}
 ]
}
